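# Physics asset for the skirt of the Digital skin of the female medium Cat Burglar body.
# Capsule bodies cover back, side and front shirt bone chains, a left frill and both thighs, with several multi-capsule bodies widening coverage.
# Angle-limited constraints link back and side chains from parent through driver to end bone, while frill and front panels use single joints.
# The thigh capsules have no constraints, suggesting they serve as leg colliders.
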PhysicsAsset /BRCosmetics/Characters/Player/Female/Medium/Bodies/F_MED_Cat_Burglar/Skin/Digital/Meshes/F_MED_Cat_Burglar_Digital_Skirt_Physics
bodies {
 dyn_driver_shirt_bck [{shapeType=Capsule, length=2.5502, radius=3.255, center=<1.2849 0 1.5>, rotation=<rot 90 0 0>}]
 dyn_driver_shirt_bck_l [{shapeType=Capsule, length=3, radius=3.255, center=<1.8945 0 1.5>, rotation=<rot 90 0 0>}]
 dyn_driver_shirt_bck_r [{shapeType=Capsule, length=3, radius=3.255, center=<1.8808 0 0.1694>, rotation=<rot 90 0 0>}]
 dyn_driver_shirt_l [{shapeType=Capsule, length=3, radius=3.255, center=<10.5316 1 1>, rotation=<rot 80 0 -0.0001>}]
 dyn_driver_shirt_r [{shapeType=Capsule, length=3, radius=3.255, center=<10.5316 -1 2>, rotation=<rot 80 0 0>}]
 dyn_frill_01_l [{shapeType=Capsule, length=2.25, radius=2.255, center=<2 0 0>, rotation=<rot 90 0 0>}]
 dyn_parent_frill_l [{shapeType=Capsule, length=0.505, radius=0.505, rotation=<rot 0 -90 0>}]
 dyn_parent_shirt_bck [{shapeType=Capsule, length=0.505, radius=0.505, rotation=<rot 0 -90 0>}]
 dyn_parent_shirt_bck_l [{shapeType=Capsule, length=0.505, radius=0.505, rotation=<rot 0 -90 0>}]
 dyn_parent_shirt_bck_r [{shapeType=Capsule, length=0.505, radius=0.505, rotation=<rot 0 -90 0>}]
 dyn_parent_shirt_l [{shapeType=Capsule, length=0.505, radius=0.505, rotation=<rot 0 -90 0>}]
 dyn_parent_shirt_r [{shapeType=Capsule, length=0.505, radius=0.505, rotation=<rot 0 -90 0>}]
 dyn_shirt_01_l [{shapeType=Capsule, length=7.75, radius=3.255, center=<8.3804 1 1>, rotation=<rot 80 -0.0001 89.9997>}]
 dyn_shirt_01_r [{shapeType=Capsule, length=6.75, radius=3.255, center=<8.3804 -1 2>, rotation=<rot 80 0.0001 90.0003>}]
 dyn_shirt_bck_01 [
  {shapeType=Capsule, length=3.738, radius=3.5, center=<5.9197 0 1.5>, rotation=<rot 90 0 0>}
  {shapeType=Capsule, length=19.5, radius=3.5, center=<10 0 1.5>, rotation=<rot 0 0 90>}
 ]
 dyn_shirt_bck_01_l [{shapeType=Capsule, length=4.4424, radius=3.5, center=<8.2418 0.0876 1.5>, rotation=<rot 90 0 0>}]
 dyn_shirt_bck_01_r [{shapeType=Capsule, length=3.3091, radius=3.5, center=<7.2882 -0.0619 -0.0853>, rotation=<rot 87.5687 0.0002 0.0003>}]
 dyn_shirt_frt_01_l [{shapeType=Capsule, length=1, radius=1}]
 dyn_shirt_frt_01_r [{shapeType=Capsule, length=1, radius=1}]
 dyn_shirt_frt_02_l [
  {shapeType=Capsule, length=5.75, radius=3.25, center=<6.9828 0 3.1793>, rotation=<rot 90 0 0>}
  {shapeType=Capsule, length=5.5625, radius=3.3956, shapeName=dyn_shirt_frt_02_l_capsule, center=<6.8602 5.8366 1.7217>, rotation=<rot 90 0 0>}
  {shapeType=Capsule, length=6.0337, radius=3.3629, shapeName=dyn_shirt_frt_02_l_capsule, center=<6.4093 -6.5223 2.1167>, rotation=<rot 90 0 0>}
 ]
 dyn_shirt_frt_02_r [
  {shapeType=Capsule, length=5.75, radius=3.25, center=<7.264 -0.0002 4.4525>, rotation=<rot 90 0 0>}
  {shapeType=Capsule, length=5.3873, radius=3.3608, shapeName=dyn_shirt_frt_02_r_capsule, center=<7.2239 6.5027 3.5925>, rotation=<rot 90 0 0>}
  {shapeType=Capsule, length=6.0844, radius=3.4179, shapeName=dyn_shirt_frt_02_r_capsule, center=<7.0311 -5.819 2.6784>, rotation=<rot 90 0 0>}
 ]
 thigh_l [
  {shapeType=Capsule, length=19.2517, radius=7.9983, center=<-12.9 0.0002 0>, rotation=<rot 90 89.4328 88.4794>}
  {shapeType=Capsule, length=15.25, radius=6.5, center=<-12.9 -4.6746 -0.0001>, rotation=<rot 90 169.3803 179.381>}
 ]
 thigh_r [
  {shapeType=Capsule, length=17.9183, radius=9.3317, center=<12.9 0.0002 0>, rotation=<rot 90 -0.0098 -0.0098>}
  {shapeType=Capsule, length=15, radius=7.75, center=<12.9 3.2189 0>, rotation=<rot 90 7.125 17.1254>}
 ]
}
constraints [
 {bone1Name=dyn_driver_shirt_bck, bone2Name=dyn_parent_shirt_bck, swing1LimitDegrees=25, swing2LimitDegrees=25, twistLimitDegrees=5}
 {bone1Name=dyn_driver_shirt_bck_r, bone2Name=dyn_parent_shirt_bck_r, swing1LimitDegrees=25, swing2LimitDegrees=25, twistLimitDegrees=5}
 {bone1Name=dyn_driver_shirt_bck_l, bone2Name=dyn_parent_shirt_bck_l, swing1LimitDegrees=25, swing2LimitDegrees=25, twistLimitDegrees=5}
 {bone1Name=dyn_shirt_bck_01, bone2Name=dyn_driver_shirt_bck, swing1LimitDegrees=25, swing2LimitDegrees=25, twistLimitDegrees=5}
 {bone1Name=dyn_shirt_bck_01_r, bone2Name=dyn_driver_shirt_bck_r, swing1LimitDegrees=25, swing2LimitDegrees=25, twistLimitDegrees=5}
 {bone1Name=dyn_shirt_bck_01_l, bone2Name=dyn_driver_shirt_bck_l, swing1LimitDegrees=25, swing2LimitDegrees=25, twistLimitDegrees=5}
 {bone1Name=dyn_shirt_01_l, bone2Name=dyn_driver_shirt_l, swing1LimitDegrees=28.4286, swing2LimitDegrees=33.5714, twistLimitDegrees=5}
 {bone1Name=dyn_driver_shirt_l, bone2Name=dyn_parent_shirt_l, swing1LimitDegrees=20, swing2LimitDegrees=20, twistLimitDegrees=5}
 {bone1Name=dyn_driver_shirt_r, bone2Name=dyn_parent_shirt_r, swing1LimitDegrees=20, swing2LimitDegrees=20, twistLimitDegrees=5}
 {bone1Name=dyn_shirt_01_r, bone2Name=dyn_driver_shirt_r, swing1LimitDegrees=28.4286, swing2LimitDegrees=33.5714, twistLimitDegrees=5}
 {bone1Name=dyn_frill_01_l, bone2Name=dyn_parent_frill_l, swing1LimitDegrees=32, swing2LimitDegrees=20, twistLimitDegrees=5}
 {bone1Name=dyn_shirt_frt_02_l, bone2Name=dyn_shirt_frt_01_l, swing1LimitDegrees=35, swing2LimitDegrees=35, twistLimitDegrees=35}
 {bone1Name=dyn_shirt_frt_02_r, bone2Name=dyn_shirt_frt_01_r, swing1LimitDegrees=35, swing2LimitDegrees=35, twistLimitDegrees=35}
]
