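# Post-process style AnimBP for the Cat Burglar Digital skin, adding secondary motion to skirt, straps and grenades.
# It copies the attached parent's pose, then chains bone constraints, two Control Rigs, pelvis-space RigidBody sims and grenade AnimDynamics, with physics weighted by DEBUG.
# Crouch, grind-rail and skydiving states drive additive corrections to straps and grenade.
# The EventGraph also derives leg-angle alphas, but their shirt and skirt-offset nodes are disconnected from the output.
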AnimBlueprint /BRCosmetics/Characters/Player/Female/Medium/Bodies/F_MED_Cat_Burglar/Skin/Digital/Meshes/F_MED_Cat_Burglar_Digital_AnimBP
target_skeleton=/BRCosmetics/Characters/Player/Female/Medium/Bodies/F_MED_Cat_Burglar/Meshes/F_MED_Cat_Burglar_Skeleton
graphs [
 AnimGraph {nodes [
   "Constraint - dyn_parent_shirt_frt_l (PPPP)"!(
    ComponentPose
    Alpha=0.3
    ConstraintWeights_0=0.0266
    ConstraintWeights_1:float=0.4551
    ConstraintWeights_2:float=0.0419
    ConstraintWeights_3:float=0.4765
   ) -> (Pose) {
    alphaBoolBlend {blendOption=Linear}
    boneToModify {boneName=dyn_parent_shirt_frt_l}
    constraintSetup [
     {targetBone {boneName=spine_01}}
     {targetBone {boneName=thigh_twist_01_l}}
     {targetBone {boneName=thigh_twist_01_r}}
     {targetBone {boneName=pelvis}}
    ]
    constraintWeights [1,1,1,1]
   }
   n1 = "Transform (Modify) Bone - Bone: dyn_shirt_frt_01_offset_l"(Alpha=Get thigh_fwd_alpha_l() -> (thigh_fwd_alpha_l:double), Translation="0,-4,6", Rotation="0,0,0", Scale="1,1,1") {
    rotationMode=BMM_Additive
    rotationSpace=BCS_BoneSpace
    translationMode=BMM_Additive
    translationSpace=BCS_BoneSpace
    alphaBoolBlend {blendOption=Linear}
    alphaScaleBiasClamp {bInterpResult=true, interpSpeedDecreasing=15, interpSpeedIncreasing=15}
    boneToModify {boneName=dyn_shirt_frt_01_offset_l}
    translation {y=-4, z=6}
   }
   n2 = "Transform (Modify) Bone - Bone: dyn_shirt_frt_01_offset_r"(ComponentPose=@n1, Alpha=Get thigh_fwd_alpha_r() -> (thigh_fwd_alpha_r:double), Translation="0,4,6", Rotation="0,0,0", Scale="1,1,1") {
    rotationMode=BMM_Additive
    rotationSpace=BCS_BoneSpace
    translationMode=BMM_Additive
    translationSpace=BCS_BoneSpace
    alphaBoolBlend {blendOption=Linear}
    alphaScaleBiasClamp {bInterpResult=true, interpSpeedDecreasing=15, interpSpeedIncreasing=15}
    boneToModify {boneName=dyn_shirt_frt_01_offset_r}
    translation {y=4, z=6}
   }
   n3 = "Transform (Modify) Bone - Bone: dyn_shirt_frt_01_offset_r"(ComponentPose=@n2, Alpha=Get legOutCrouch_R() -> (legOutCrouch_R:double), Translation="0,3,0.000", Rotation="0,0,0", Scale="1,1,1") {
    rotationMode=BMM_Additive
    rotationSpace=BCS_BoneSpace
    translationMode=BMM_Additive
    translationSpace=BCS_BoneSpace
    alphaBoolBlend {blendOption=Linear}
    alphaScaleBiasClamp {bInterpResult=true, interpSpeedDecreasing=15, interpSpeedIncreasing=15}
    boneToModify {boneName=dyn_shirt_frt_01_offset_r}
    translation {y=3}
   }
   n4 = "Transform (Modify) Bone - Bone: skirt_offset_r"(ComponentPose=@n3, Alpha=Get legOutNoCrouch_R() -> (legOutNoCrouch_R:double), Translation="0.000,-5,10", Rotation="0,0,0", Scale="1,1,1") {
    rotationMode=BMM_Additive
    rotationSpace=BCS_BoneSpace
    translationMode=BMM_Additive
    translationSpace=BCS_BoneSpace
    alphaBoolBlend {blendOption=Linear}
    alphaScaleBiasClamp {interpSpeedDecreasing=15, interpSpeedIncreasing=15}
    boneToModify {boneName=skirt_offset_r}
    translation {y=-5, z=10}
   }
   n5 = "Transform (Modify) Bone - Bone: dyn_driver_shirt_bck_r"(ComponentPose=@n4, Alpha=Get legBack_R() -> (legBack_R:double), Translation="0,-0.023,3", Rotation="20,0,0", Scale="1,1,1") {
    rotationMode=BMM_Additive
    rotationSpace=BCS_BoneSpace
    translationMode=BMM_Additive
    translationSpace=BCS_BoneSpace
    alphaBoolBlend {blendOption=Linear}
    alphaScaleBiasClamp {interpSpeedDecreasing=15, interpSpeedIncreasing=15}
    boneToModify {boneName=dyn_driver_shirt_bck_r}
    rotation {pitch=20}
    translation {y=-0.023, z=3}
   }
   "Transform (Modify) Bone - Bone: dyn_driver_shirt_bck_l"(ComponentPose=@n5, Alpha=Get legBack_L() -> (legBack_L:double), Translation="0.000,0.023,3", Rotation="20.000,0.000,0", Scale="1,1,1") {
    rotationMode=BMM_Additive
    rotationSpace=BCS_BoneSpace
    translationMode=BMM_Additive
    translationSpace=BCS_BoneSpace
    alphaBoolBlend {blendOption=Linear}
    alphaScaleBiasClamp {interpSpeedDecreasing=15, interpSpeedIncreasing=15}
    boneToModify {boneName=dyn_driver_shirt_bck_l}
    rotation {pitch=20}
    translation {y=0.023, z=3}
   }
   n6 = "Transform (Modify) Bone - Bone: dyn_parent_shirt_l"(Alpha=1, Translation="0,1,1", Rotation="0,0,0", Scale="1,1,1") {
    rotationMode=BMM_Additive
    rotationSpace=BCS_BoneSpace
    translationMode=BMM_Additive
    translationSpace=BCS_BoneSpace
    alphaBoolBlend {blendOption=Linear}
    boneToModify {boneName=dyn_parent_shirt_l}
    translation {y=1, z=1}
   }
   n7 = "Transform (Modify) Bone - Bone: dyn_parent_shirt_r"(ComponentPose=@n6, Alpha=1, Translation="0,1,1", Rotation="0,0,0", Scale="1,1,1") {
    rotationMode=BMM_Additive
    rotationSpace=BCS_BoneSpace
    translationMode=BMM_Additive
    translationSpace=BCS_BoneSpace
    alphaBoolBlend {blendOption=Linear}
    boneToModify {boneName=dyn_parent_shirt_r}
    translation {y=1, z=1}
   }
   n8 = "Transform (Modify) Bone - Bone: dyn_driver_shirt_frt_l"(ComponentPose=@n7, Alpha=Get thigh_fwd_alpha_l(), Translation="0.7,-3,6", Rotation="0,0,0", Scale="1,1,1") {
    rotationMode=BMM_Additive
    rotationSpace=BCS_BoneSpace
    translationMode=BMM_Additive
    translationSpace=BCS_BoneSpace
    alphaBoolBlend {blendOption=Linear}
    boneToModify {boneName=dyn_driver_shirt_frt_l}
    translation=<0.7 -3 6>
   }
   "Transform (Modify) Bone - Bone: dyn_driver_shirt_frt_r"(ComponentPose=@n8, Alpha=Get thigh_fwd_alpha_r(), Translation="0.7,3,6", Rotation="0,0,0", Scale="1,1,1") {
    rotationMode=BMM_Additive
    rotationSpace=BCS_BoneSpace
    translationMode=BMM_Additive
    translationSpace=BCS_BoneSpace
    alphaBoolBlend {blendOption=Linear}
    boneToModify {boneName=dyn_driver_shirt_frt_r}
    translation=<0.7 3 6>
   }
   n9 = "Transform (Modify) Bone - Bone: dyn_strap_bck_01"(Alpha=1, Translation="0,0,0", Rotation="0,0,0", Scale="1,1,1") {
    rotationSpace=BCS_BoneSpace
    translationMode=BMM_Replace
    translationSpace=BCS_ParentBoneSpace
    alphaBoolBlend {blendOption=Linear}
    alphaScaleBiasClamp {interpSpeedDecreasing=15, interpSpeedIncreasing=15}
    boneToModify {boneName=dyn_strap_bck_01}
   }
   "Transform (Modify) Bone - Bone: dyn_strap_frt_01"(ComponentPose=@n9, Alpha=1, Translation="0,0,0", Rotation="0,0,0", Scale="1,1,1") {
    rotationSpace=BCS_BoneSpace
    translationMode=BMM_Replace
    translationSpace=BCS_ParentBoneSpace
    alphaBoolBlend {blendOption=Linear}
    alphaScaleBiasClamp {interpSpeedDecreasing=15, interpSpeedIncreasing=15}
    boneToModify {boneName=dyn_strap_frt_01}
   }
   n10 = "Constraint - dyn_thigh_shirt_l (P)"(ComponentPose=Local To Component(LocalPose=Copy Pose From Mesh() {bUseAttachedParent=true}), Alpha=1, ConstraintWeights_0=1) {
    alphaBoolBlend {blendOption=Linear}
    boneToModify {boneName=dyn_thigh_shirt_l}
    constraintSetup [{targetBone {boneName=thigh_twist_01_l}}]
    constraintWeights [1]
   }
   n11 = "Constraint - dyn_thigh_shirt_r (P)"(ComponentPose=@n10, Alpha=1, ConstraintWeights_0=1) {
    alphaBoolBlend {blendOption=Linear}
    boneToModify {boneName=dyn_thigh_shirt_r}
    constraintSetup [{targetBone {boneName=thigh_twist_01_r}}]
    constraintWeights [1]
   }
   n12 = "Constraint - dyn_parent_frill_l (P)"(ComponentPose=@n11, Alpha=1, ConstraintWeights_0=1) {
    alphaBoolBlend {blendOption=Linear}
    boneToModify {boneName=dyn_parent_frill_l}
    constraintSetup [{targetBone {boneName=thigh_l}}]
    constraintWeights [1]
   }
   n13 = Control Rig(Source=Component To Local(ComponentPose=@n12)) {
    alphaBoolBlend {blendOption=Linear}
    controlRigAssetReference {blueprintRigClass=&/F_MED_Cat_Burglar_Digital_Skirt_CtrlRig.F_MED_Cat_Burglar_Digital_Skirt_CtrlRig_C}
   }
   n14 = RigidBody!(
    ComponentPose=Local To Component(LocalPose=@n13)
    Alpha=Get DEBUG() -> (DEBUG:double)
    OverrideWorldGravity:Vector=Get Gravity Override()
    ExternalForce
   ) -> (Pose) {
    bClampLinearTranslationLimitToRefPose=true
    bOverrideWorldGravity=true
    overridePhysicsAsset=&/F_MED_Cat_Burglar_Digital_Skirt_Physics
    simulationSpace=BaseBoneSpace
    alphaBoolBlend {blendOption=Linear}
    alphaScaleBiasClamp {interpSpeedIncreasing=2}
    baseBoneRef {boneName=pelvis}
    componentAppliedLinearAccClamp=<3000 3000 3000>
    componentLinearAccScale=<0.4 0.3 0.3>
    componentLinearVelScale=<0.4 0.3 0.3>
    simSpaceSettings {
     maxAngularAcceleration=1000
     maxAngularVelocity=3
     maxLinearAcceleration=700
     maxLinearVelocity=500
     velocityScaleZ=0
     worldAlpha=0.25
     externalAngularVelocity {z=0.1}
     externalLinearDragV=<0.2 0.2>
    }
   }
   n15 = "Constraint - dyn_parent_strap_frt_01 (P)"(ComponentPose=@n14, Alpha=1, ConstraintWeights_0=1) {
    alphaBoolBlend {blendOption=Linear}
    boneToModify {boneName=dyn_parent_strap_frt_01}
    constraintSetup [{targetBone {boneName=pelvis}}]
    constraintWeights [1]
   }
   n16 = "Constraint - dyn_parent_strap_bck_01 (P)"(ComponentPose=@n15, Alpha=1, ConstraintWeights_0=1) {
    alphaBoolBlend {blendOption=Linear}
    boneToModify {boneName=dyn_parent_strap_bck_01}
    constraintSetup [{targetBone {boneName=pelvis}}]
    constraintWeights [1]
   }
   n17 = "Constraint - dyn_parent_strap_01 (P)"(ComponentPose=@n16, Alpha=1, ConstraintWeights_0=1) {
    alphaBoolBlend {blendOption=Linear}
    boneToModify {boneName=dyn_parent_strap_01}
    constraintSetup [{targetBone {boneName=pelvis}}]
    constraintWeights [1]
   }
   n18 = "Transform (Modify) Bone - Bone: dyn_strap_bck_01"(ComponentPose=@n17, Alpha="To Float (Boolean)"(InBool=Get bIsSkyDiving()), Translation="0,0,0", Rotation="20,-25,0", Scale="1,1,1") {
    rotationMode=BMM_Additive
    rotationSpace=BCS_BoneSpace
    translationMode=BMM_Additive
    alphaBoolBlend {blendOption=Linear}
    boneToModify {boneName=dyn_strap_bck_01}
    rotation {pitch=20, yaw=-25}
   }
   n19 = Get Is Crouching Float() -> (Is Crouching Float:float)
   n20 = "Transform (Modify) Bone - Bone: dyn_parent_strap_frt_01"(ComponentPose=@n18, Alpha=@n19, Translation="0,0,0", Rotation="-90,-90,60", Scale="1,1,1") {
    rotationMode=BMM_Additive
    rotationSpace=BCS_BoneSpace
    translationMode=BMM_Additive
    translationSpace=BCS_BoneSpace
    alphaBoolBlend {blendOption=Linear}
    boneToModify {boneName=dyn_parent_strap_frt_01}
    rotation=<rot -90 -90 60>
   }
   n21 = "Transform (Modify) Bone - Bone: dyn_driver_strap_01"(ComponentPose=@n20, Alpha=@n19, Translation="0,0,0", Rotation="5,-10,20", Scale="1,1,1") {
    rotationMode=BMM_Additive
    rotationSpace=BCS_BoneSpace
    translationMode=BMM_Additive
    translationSpace=BCS_BoneSpace
    alphaBoolBlend {blendOption=Linear}
    boneToModify {boneName=dyn_driver_strap_01}
    rotation=<rot 5 -10 20>
   }
   n22 = Get Is on GrindRail Float() -> (Is on GrindRail Float:float)
   n23 = "Transform (Modify) Bone - Bone: dyn_parent_strap_frt_01"(ComponentPose=@n21, Alpha=@n22, Translation="0,0,-1", Rotation="-90,-90,60", Scale="1,1,1") {
    rotationMode=BMM_Additive
    rotationSpace=BCS_BoneSpace
    translationMode=BMM_Additive
    translationSpace=BCS_BoneSpace
    alphaBoolBlend {blendOption=Linear}
    boneToModify {boneName=dyn_parent_strap_frt_01}
    rotation=<rot -90 -90 60>
    translation {z=-1}
   }
   n24 = "Transform (Modify) Bone - Bone: dyn_driver_strap_01"(ComponentPose=@n23, Alpha=@n22, Translation="0,0,0", Rotation="5,-10,0", Scale="1,1,1") {
    rotationMode=BMM_Additive
    rotationSpace=BCS_BoneSpace
    translationMode=BMM_Additive
    translationSpace=BCS_BoneSpace
    alphaBoolBlend {blendOption=Linear}
    boneToModify {boneName=dyn_driver_strap_01}
    rotation {pitch=5, yaw=-10}
   }
   n25 = RigidBody(ComponentPose=@n24, Alpha=Get DEBUG(), OverrideWorldGravity=Get Gravity Override()) {
    bClampLinearTranslationLimitToRefPose=true
    bOverrideWorldGravity=true
    overridePhysicsAsset=&/F_MED_Cat_Burglar_Digital_Strap_Physics
    simulationSpace=BaseBoneSpace
    alphaBoolBlend {blendOption=Linear}
    alphaScaleBiasClamp {interpSpeedIncreasing=2}
    baseBoneRef {boneName=pelvis}
    componentAppliedLinearAccClamp=<3000 3000 3000>
    componentLinearAccScale=<0.4 0.3 0.3>
    componentLinearVelScale=<0.4 0.3 0.3>
    simSpaceSettings {
     maxAngularAcceleration=1000
     maxAngularVelocity=3
     maxLinearAcceleration=700
     maxLinearVelocity=500
     velocityScaleZ=0
     worldAlpha=0.25
     externalAngularVelocity {z=0.1}
     externalLinearDragV=<0.2 0.2>
    }
   }
   n26 = "Constraint - grenade_1 (PPPP)"!(
    ComponentPose=@n25
    Alpha=1
    ConstraintWeights_0=0.6214
    ConstraintWeights_1:float=0.2028
    ConstraintWeights_2:float=0.1369
    ConstraintWeights_3:float=0.0389
   ) -> (Pose) {
    alphaBoolBlend {blendOption=Linear}
    boneToModify {boneName=grenade_1}
    constraintSetup [{targetBone {boneName=pelvis}}, {targetBone {boneName=spine_02}}, {targetBone {boneName=spine_01}}, {targetBone {boneName=spine_04}}]
    constraintWeights [1,1,1,1]
   }
   n27 = "Constraint - grenade_2 (PPPP)"!(
    ComponentPose=@n26
    Alpha=1
    ConstraintWeights_0=0.661
    ConstraintWeights_1:float=0.1741
    ConstraintWeights_2:float=0.152
    ConstraintWeights_3:float=0.0129
   ) -> (Pose) {
    alphaBoolBlend {blendOption=Linear}
    boneToModify {boneName=grenade_2}
    constraintSetup [{targetBone {boneName=pelvis}}, {targetBone {boneName=spine_02}}, {targetBone {boneName=spine_01}}, {targetBone {boneName=spine_04}}]
    constraintWeights [1,1,1,1]
   }
   n28 = Get DEBUG()
   n29 = Anim Dynamics - Bone: grenade_1(ComponentPose=@n27, Alpha=@n28, ExternalForce="0,0,0") {
    angularDampingOverride=1
    angularSpringConstant=1500
    bAngularSpring=true
    bGravityOverrideInSimSpace=true
    bOverrideAngularDamping=true
    gravityScale=0.35
    simulationSpace=RootRelative
    alphaBoolBlend {blendOption=Linear}
    alphaScaleBiasClamp {scale=0.3}
    boundBone {boneName=grenade_1}
    physicsBodyDefinitions [{
      sphereCollisionRadius=0
      boundBone {boneName=grenade_1}
      boxExtents=<10 20 20>
      constraintSetup {angularLimitsMax=<0 30 0>, angularLimitsMin=<0 -30 -90>, angularTarget=<0 0 0>}
      localJointOffset=<10 0 0>
     }]
   }
   n30 = Anim Dynamics - Bone: grenade_2(ComponentPose=@n29, Alpha=@n28, ExternalForce="0,0,0") {
    angularDampingOverride=1
    angularSpringConstant=1500
    bAngularSpring=true
    bGravityOverrideInSimSpace=true
    bOverrideAngularDamping=true
    gravityScale=0.35
    simulationSpace=RootRelative
    alphaBoolBlend {blendOption=Linear}
    alphaScaleBiasClamp {scale=0.3}
    boundBone {boneName=grenade_2}
    physicsBodyDefinitions [{
      sphereCollisionRadius=0
      boundBone {boneName=grenade_2}
      boxExtents=<10 20 20>
      constraintSetup {angularLimitsMax=<0 30 0>, angularLimitsMin=<0 -30 -90>, angularTarget=<0 0 0>}
      localJointOffset=<10 0 0>
     }]
   }
   n31 = "Transform (Modify) Bone - Bone: grenade_1"(ComponentPose=@n30, Alpha=Get Is Crouching Float(), Translation="0,0,0", Rotation="0,0,10", Scale="1,1,1") {
    rotationMode=BMM_Additive
    rotationSpace=BCS_BoneSpace
    translationMode=BMM_Additive
    translationSpace=BCS_BoneSpace
    alphaBoolBlend {blendOption=Linear}
    boneToModify {boneName=grenade_1}
    rotation {roll=10}
   }
   n32 = "Transform (Modify) Bone - Bone: grenade_1"(ComponentPose=@n31, Alpha=Get Is on GrindRail Float(), Translation="0,0,0", Rotation="0,0,10", Scale="1,1,1") {
    rotationMode=BMM_Additive
    rotationSpace=BCS_BoneSpace
    translationMode=BMM_Additive
    translationSpace=BCS_BoneSpace
    alphaBoolBlend {blendOption=Linear}
    boneToModify {boneName=grenade_1}
    rotation {roll=10}
   }
   n33 = Control Rig(Source=Component To Local(ComponentPose=@n32)) {
    alphaBoolBlend {blendOption=Linear}
    controlRigAssetReference {blueprintRigClass=&/F_MED_Cat_Burglar_Digital_CtrlRig.F_MED_Cat_Burglar_Digital_CtrlRig_C}
   }
   Output Pose(Result=@n33) {layerGroup=None}
  ]}
 EventGraph {nodes [
   K2Node_Event_1 = CustomEvent() {is=title K2Node_CustomEvent}
   > Set Skel_Mesh(Skel_Mesh=GetOwnerSkeletalMeshComponent())
   Event BlueprintUpdateAnimation()
   > Get!(execute:Exec) -> (then:Exec, else:Exec, Body_Mesh:Skeletal Mesh Component Object Reference)
   > n1 = Sequence!(execute) -> (then_0, then_1, then_2:Exec, then_3:Exec, then_4:Exec, then_5:Exec)
   > n2 = Branch(Condition=Get bIsCrouching())
   n3 = Get PawnSpeed2D()
   > n4 = Branch(Condition="float <= float"!(A:double=MapRangeClamped(Value=@n3, InRangeA=0, InRangeB=150), B:double=0) -> (ReturnValue:bool) {is=title K2Node_CallFunction})
   > Set crouch_speed_alpha(crouch_speed_alpha:double=1) -> (Output_Get:double)
   > Set crouch_moving(crouch_moving:bool=false) -> (Output_Get:bool)
   > Set crouch_rigid_alpha(crouch_rigid_alpha:double=0.2) -> (Output_Get:double)
   > Set move_speed_alpha(move_speed_alpha:double=0) -> (Output_Get:double)
   Set crouch_speed_alpha(execute=@n4.else, crouch_speed_alpha=0)
   > Set crouch_moving(crouch_moving=true)
   > Set crouch_rigid_alpha(crouch_rigid_alpha=0.55)
   > Set move_speed_alpha(move_speed_alpha=1)
   Set crouch_rigid_alpha(execute=@n2.else, crouch_rigid_alpha=0.8)
   Set Is on GrindRail Float(execute=@n1.then_1, Is on GrindRail Float:float="To Float (Boolean)"(InBool=Get bIsOnGrindRail())) -> (Output_Get:float)
   Set Is Crouching Float(execute=@n1.then_3, Is Crouching Float:float="To Float (Boolean)"(InBool=Get bIsCrouching())) -> (Output_Get:float)
   n5 = Sequence!(execute=@n1.then_4) -> (then_0, then_1, then_2, then_3, then_4)
   n6 = TryGetFortPawn()
   Is Valid(exec=@n5.then_2, InputObject=@n6)
   n7 = Get bIsDBNO!(self:Fort Pawn Object Reference=@n6) -> (bIsDBNO)
   > Set IsDBNO(IsDBNO:bool=@n7) -> (Output_Get:bool)
   > Set NotDBNO(NotDBNO:bool=NOT Boolean(A=@n7)) -> (Output_Get:bool)
   > n8 = Sequence!(execute) -> (then_0, then_1)
   n9 = Get bIsCrouching()
   > Set crouchAlpha(crouchAlpha:double="To Float (Boolean)"(InBool=@n9)) -> (Output_Get:double)
   n10 = Branch(execute=@n8.then_1, Condition=@n9)
   n11 = Get PawnSpeed2D()
   > n12 = Branch(Condition="float <= float"(A=MapRangeClamped(Value=@n11, InRangeA=0, InRangeB=150), B=0))
   > Set crouch_speed_alpha(crouch_speed_alpha=1)
   > Set crouch_moving(crouch_moving=false)
   > Set crouch_rigid_alpha(crouch_rigid_alpha=0.2)
   > Set move_speed_alpha(move_speed_alpha=0)
   Set crouch_speed_alpha(execute=@n12.else, crouch_speed_alpha=0)
   > Set crouch_moving(crouch_moving=true)
   > Set crouch_rigid_alpha(crouch_rigid_alpha=0.55)
   > Set move_speed_alpha(move_speed_alpha=1)
   Set crouch_rigid_alpha(execute=@n10.else, crouch_rigid_alpha=0.8)
   n13 = "Absolute (Float)"(A=GetOwnerDirection!(self:Custom Character Part Anim Instance Object Reference) -> (ReturnValue:float))
   Set isMovingSide(
    execute=@n5.then_4
    isMovingSide:bool=AND Boolean(
     A="float > float"(A=@n13, B=60) {is=title K2Node_CallFunction}
     B="float < float"!(A:double=@n13, B:double=120) -> (ReturnValue:bool) {is=title K2Node_CallFunction}
    )
   ) -> (Output_Get:bool)
   n14 = GetOwnerDirection()
   n15 = Branch(execute=@n5.then_1, Condition=AND Boolean(A="float > float"(A=@n14, B=60), B="float < float"(A=@n14, B=120)))
   n16 = Get Body_Mesh()
   n17 = BreakTransform(InTransform=GetDeltaTransformFromRefPose(self=@n16, BoneName=thigh_l, BaseName=pelvis))
   > Set moveRighttLegCrossAlpha(moveRighttLegCrossAlpha:double=MapRangeClamped(Value=BreakRotator(InRot=@n17.Rotation).Pitch, InRangeA=-15, InRangeB=0, OutRangeA=1, OutRangeB=0)) -> (Output_Get:double)
   Set moveRighttLegCrossAlpha(execute=@n15.else, moveRighttLegCrossAlpha=0)
   n18 = Branch(execute=@n5.then_0, Condition=AND Boolean(A="float > float"(A=@n14, B=-120), B="float < float"(A=@n14, B=-60)))
   n19 = BreakTransform(InTransform=GetDeltaTransformFromRefPose(self=@n16, BoneName=thigh_r, BaseName=pelvis))
   > Set moveLeftLegCrossAlpha(moveLeftLegCrossAlpha:double=MapRangeClamped(Value=BreakRotator(InRot=@n19.Rotation).Pitch, InRangeA=-15, InRangeB=0, OutRangeA=1, OutRangeB=0)) -> (Output_Get:double)
   Set moveLeftLegCrossAlpha(execute=@n18.else, moveLeftLegCrossAlpha=0)
   Set jump_alpha(execute=@n5.then_3, jump_alpha:double=MapRangeClamped(Value=Get PawnVelocityZ(), InRangeA=-700, InRangeB=0, OutRangeA=1, OutRangeB=0)) -> (Output_Get:double)
   > n20 = Branch()
   > Set isSkyDivingOn(isSkyDivingOn:double="To Float (Boolean)"(InBool=Get bIsSkyDiving())) -> (Output_Get:double)
   Set isSkyDivingOn(execute=@n20.else, isSkyDivingOn=0)
   n21 = Sequence!(execute=@n1.then_5) -> (then_0, then_1, then_2:Exec, then_3:Exec)
   > Set thigh_fwd_alpha_r(thigh_fwd_alpha_r:double=GetAlphaFromReferenceTransformDelta(
     SkeletalMeshComponent=Get Skel_Mesh()
     BoneName=thigh_r
     BaseName=pelvis
     DeltaComponentType=QuaternionTwist
     InputRange=(X=0,Y=100)
     TwistAxis="0,0,1"
    )) -> (Output_Get:double)
   n22 = Get Skel_Mesh()
   Set thigh_back_alpha_l(
    execute=@n21.then_2
    thigh_back_alpha_l:double=GetAlphaFromReferenceTransformDelta(SkeletalMeshComponent=@n22, BoneName=thigh_l, BaseName=pelvis, DeltaComponentType=QuaternionTwist, InputRange=(X=0,Y=-60), TwistAxis="0,0,1")
   ) -> (Output_Get:double)
   Set thigh_fwd_alpha_l(
    execute=@n21.then_1
    thigh_fwd_alpha_l:double=GetAlphaFromReferenceTransformDelta(SkeletalMeshComponent=@n22, BoneName=thigh_l, BaseName=pelvis, DeltaComponentType=QuaternionTwist, InputRange=(X=0,Y=108), TwistAxis="0,0,1")
   ) -> (Output_Get:double)
   Set arm_rot_high_r(
    execute=@n21.then_3
    arm_rot_high_r:double=GetAlphaFromReferenceTransformDelta(SkeletalMeshComponent=Get Skel_Mesh(), BoneName=upperarm_r, BaseName=clavicle_r, InputRange=(X=30,Y=90), TwistAxis="0,0,1")
   ) -> (Output_Get:double)
   Set legUp_L(
    execute=@n1.then_2
    legUp_L:double=GetAlphaFromReferenceTransformDelta(
     SkeletalMeshComponent=Get Body_Mesh()
     BoneName=thigh_l
     BaseName=pelvis
     DeltaComponentType=QuaternionTwist
     InputRange=(X=25,Y=95)
     TwistAxis="0,0,1"
    )
   ) -> (Output_Get:double)
   > Set legUp_R(legUp_R:double=GetAlphaFromReferenceTransformDelta(
     SkeletalMeshComponent=Get Body_Mesh()
     BoneName=thigh_r
     BaseName=pelvis
     DeltaComponentType=QuaternionTwist
     InputRange=(X=25,Y=95)
     TwistAxis="0,0,1"
    )) -> (Output_Get:double)
   > n23 = Branch(Condition=Get is_crouching() -> (is_crouching:bool))
   > Set legBack_L(legBack_L:double=GetAlphaFromReferenceTransformDelta(
     SkeletalMeshComponent=Get Body_Mesh()
     BoneName=thigh_l
     BaseName=pelvis
     DeltaComponentType=QuaternionTwist
     InputRange=(X=50,Y=20)
     TwistAxis="0,0,1"
    )) -> (Output_Get:double)
   > Set legBack_R(legBack_R:double=GetAlphaFromReferenceTransformDelta(
     SkeletalMeshComponent=Get Body_Mesh()
     BoneName=thigh_r
     BaseName=pelvis
     DeltaComponentType=QuaternionTwist
     InputRange=(X=50,Y=20)
     TwistAxis="0,0,1"
    )) -> (Output_Get:double)
   > Set legOutCrouch_R(legOutCrouch_R:double=GetAlphaFromReferenceTransformDelta(
     SkeletalMeshComponent=Get Body_Mesh()
     BoneName=thigh_r
     BaseName=pelvis
     DeltaComponentType=QuaternionTwist
     InputRange=(X=0,Y=20)
     TwistAxis="0,1,0"
    )) -> (Output_Get:double)
   > Set legOutNoCrouch_R(legOutNoCrouch_R:double=0) -> (Output_Get:double)
   Set legBack_L(
    execute=@n23.else
    legBack_L=GetAlphaFromReferenceTransformDelta(
     SkeletalMeshComponent=Get Body_Mesh()
     BoneName=thigh_l
     BaseName=pelvis
     DeltaComponentType=QuaternionTwist
     InputRange=(X=20,Y=-11)
     TwistAxis="0,0,1"
    )
   )
   > Set legBack_R(legBack_R=GetAlphaFromReferenceTransformDelta(
     SkeletalMeshComponent=Get Body_Mesh()
     BoneName=thigh_r
     BaseName=pelvis
     DeltaComponentType=QuaternionTwist
     InputRange=(X=10,Y=-8)
     TwistAxis="0,0,1"
    ))
   > Set legOutNoCrouch_R(legOutNoCrouch_R=GetAlphaFromReferenceTransformDelta(
     SkeletalMeshComponent=Get Body_Mesh()
     BoneName=thigh_r
     BaseName=pelvis
     DeltaComponentType=QuaternionTwist
     InputRange=(X=90,Y=30)
     TwistAxis="0,1,0"
    ))
   > Set legOutCrouch_R(legOutCrouch_R=0)
   Event OnBeginCrouching()
   > Set is_crouching(is_crouching:bool=true) -> (Output_Get:bool)
   > Set crouch_alpha(crouch_alpha:double=1) -> (Output_Get:double)
   Event OnEndCrouching()
   > Set is_crouching(is_crouching=false)
   > Set crouch_alpha(crouch_alpha=0)
   Event BlueprintInitializeAnimation()
   > n24 = Gravity Override State Function DEPRECATED(CharacterPartAnimInstance=Self-Reference(), GameTimeInSeconds=GetGameTimeInSeconds())
   > Set Gravity Override(Gravity Override=@n24)
   n25 = GetOwnerSkeletalMeshComponent()
   > Set Body_Mesh(Body_Mesh=@n25)
   > Set Pawn(Pawn:Fort Pawn Object Reference=TryGetFortPawn()) -> (Output_Get:Fort Pawn Object Reference)
   > Set Skel_Mesh(Skel_Mesh=@n25)
   TryGetPawnOwner()
  ]}
]
variables [
 arm_rot_high_r {type=number}
 Body_Mesh {type=SkeletalMeshComponent}
 crouch_alpha {type=number}
 crouch_moving {type=boolean}
 crouch_rigid_alpha {type=number}
 crouch_speed_alpha {type=number}
 crouchAlpha {type=number}
 DEBUG {type=number, default=1}
 DEBUG_1 {type=number}
 Gravity Override {type=Vector}
 Is Crouching Float {type=number}
 Is on GrindRail Float {type=number}
 is_crouching {type=boolean}
 IsDBNO {type=boolean}
 isMovingSide {type=boolean}
 isSkyDivingOn {type=number}
 jump_alpha {type=number}
 legBack_L {type=number}
 legBack_R {type=number}
 legOutCrouch_R {type=number}
 legOutNoCrouch_R {type=number}
 legUp_L {type=number}
 legUp_R {type=number}
 move_speed_alpha {type=number}
 moveLeftLegCrossAlpha {type=number}
 moveRighttLegCrossAlpha {type=number}
 NotDBNO {type=boolean}
 Pawn {type=FortPawn, default=null}
 Skel_Mesh {type=SkeletalMeshComponent}
 thigh_back_alpha_l {type=number}
 thigh_fwd_alpha_l {type=number}
 thigh_fwd_alpha_r {type=number}
]
